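# Masked two-sided mesh-particle material for the banana skydive contrail, animating peeling, core slip-out, spoilage and dissolve through Dynamic Parameters.
# A U-coordinate split separates peel from core for vertex rotations: the peel bends along V with a time-based sine wobble, while the core rotates about an offset pivot and lifts.
# Color blends fresh to bad textures, tinting peel interiors via TwoSidedSign, times particle color; swirly noise drives the dissolve.
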Material /BRCosmetics/Athena/Effects/Materials/Skydiving/P_Contrail_Banana/M_Skydive_Banana
graph {
 properties {
  bUsedWithMeshParticles=true
  bUsedWithNiagaraMeshParticles=true
  blendMode=BLEND_Masked
  pixelDepthOffsetMode=PDOM_Legacy
  twoSided=true
  usedShadingModels=Default Lit
 }
 nodes [
  n1 = DynamicParameter() {paramNames ["StripOff(0-1)", Core_slip_out, "Fresh_level(1-0)", Dissovle]}
  n2 = Multiply(A=@n1.Dissovle, B=Constant() {r=2}) {constA=1}
  n3 = Add(A=@n2)
  n4 = Constant() {r=2}
  n5 = Subtract(A=@n3, B=@n4)
  n6 = TextureCoordinate()
  n7 = Constant2Vector() {g=0.5, r=0.5}
  LinearInterpolate(
   A=Add(A=Multiply(A=@n2) {constB=-1.2}, B=OneMinus(None=@n2))
   B=If(A=@n3, B=@n4, "A > B"=OneMinus(None=@n5), "A == B"=@n3, "A < B"=@n3)
   Alpha=TextureSample(UVs=Multiply(A=@n6, B=@n7)) {samplerType=SAMPLERTYPE_LinearColor, texture=/Game/Weapons/WeaponSkins/Textures/T_Swirly_Noise_highContrast1}.R
  )
  n8 = TextureCoordinate()
  n9 = Constant() {r=1}
  n10 = If(A=ComponentMask(None=@n8) {r=true}, B=Constant() {r=0.495}, "A > B"=Constant(), "A == B"=@n9, "A < B"=@n9)
  n11 = VectorParameter() {parameterName=Peel_inside_color, defaultValue=<rgba 0.7708 0.6006 0.3292 1>}
  n12 = OneMinus(None=@n10)
  n13 = TextureSample(UVs=@n8) {texture=/BRCosmetics/Athena/Effects/Textures/Skydive/P_Contrail_Banana/T_Skydive_banana_fresh}
  n14 = LinearInterpolate(A=Add(A=Multiply(A=@n10, B=@n11.RGB), B=Multiply(A=@n12, B=@n13.RGB)), B=@n13.RGB, Alpha=TwoSidedSign())
  n15 = TextureCoordinate()
  n16 = Constant() {r=1}
  n17 = If(A=ComponentMask(None=@n15) {r=true}, B=Constant() {r=0.495}, "A > B"=Constant(), "A == B"=@n16, "A < B"=@n16)
  n18 = VectorParameter() {parameterName=Peel_inside_bad_color, defaultValue=<rgba 0.6198 0.4628 0.3281 1>}
  n19 = OneMinus(None=@n17)
  n20 = TextureSample(UVs=@n15) {texture=/BRCosmetics/Athena/Effects/Textures/Skydive/P_Contrail_Banana/T_Skydive_banana_bad}
  n21 = LinearInterpolate(A=Add(A=Multiply(A=@n17, B=@n18.RGB), B=Multiply(A=@n19, B=@n20.RGB)), B=@n20.RGB, Alpha=TwoSidedSign())
  n22 = OneMinus(None=@n1."Fresh_level(1-0)")
  MaterialFunctionCall(Emissive=Multiply(A=LinearInterpolate(A=@n14, B=@n21, Alpha=@n22), B=ParticleColor().RGB)) {materialFunction=&MF_FX_EmissiveAdjustments}
  n23 = Constant3Vector() {constant=<rgba 0 1 0 0>}
  n24 = Normalize(VectorInput=CrossProduct(A=VertexNormalWS(), B=Transform(None=@n23) {transformSourceType=TRANSFORMSOURCE_Local}))
  n25 = ScalarParameter() {defaultValue=-30, parameterName=Angle}
  n26 = Multiply(A=Sine(None=Time()) {period=0.5}) {constB=0.3}
  n27 = OneMinus(None=ComponentMask(None=TextureCoordinate()) {g=true})
  n28 = Multiply(A=Multiply(A=Multiply(A=@n25, B=@n1."StripOff(0-1)")) {constB=0.01}, B=Multiply(A=Add(A=@n26), B=@n27))
  n29 = ParticlePositionWS()
  n30 = WorldPosition()
  n31 = LinearInterpolate(Alpha=Clamp(None=LinearInterpolate(Alpha=@n27) {constA=-1})) {constB=0.25}
  n32 = Multiply(
   A=@n10
   B=Multiply(A=RotateAboutAxis(NormalizedRotationAxis=@n24, RotationAngle=@n28, PivotPoint=@n29, Position=@n30.XYZ), B=Sine(None=@n31))
  )
  n33 = OneMinus(None=@n10)
  n34 = Multiply(A=Multiply(A=ScalarParameter() {defaultValue=15, parameterName=Angle_core}, B=@n1.Core_slip_out)) {constB=0.01}
  n35 = VectorParameter() {parameterName=rotate_pivot, defaultValue=<rgba 0 -17.8783 8.0993 1>}
  n36 = Multiply(
   A=@n33
   B=RotateAboutAxis(NormalizedRotationAxis=@n24, RotationAngle=Multiply(A=@n34), PivotPoint=Add(A=@n29, B=@n35.RGB) {constB=50}, Position=@n30.XYZ)
  )
  n37 = AppendVector(A=Constant2Vector(), B=Multiply(A=ScalarParameter() {defaultValue=2, parameterName=Additional_hight}, B=@n1.Core_slip_out))
  Reroute(None=Add(A=Add(A=@n32, B=@n36), B=Multiply(A=@n33, B=@n37)) {constB=10})
 ]
}
