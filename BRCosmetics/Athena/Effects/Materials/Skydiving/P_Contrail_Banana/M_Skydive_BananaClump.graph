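# Lit mesh-particle material for the banana clump in a skydive contrail, sampling the shared tropical fruit diffuse and normal textures.
# Builds a vibration offset from sine/cosine vectors driven by Time, phase-shifted by object position Z.
# UV1 acts as a mask: U scales the motion, while a V threshold splits the mesh between two differently phased, differently scaled waves.
# Dynamic parameters set vibration speed and strength per particle.
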Material /BRCosmetics/Athena/Effects/Materials/Skydiving/P_Contrail_Banana/M_Skydive_BananaClump
graph {
 properties {bUsedWithMeshParticles=true, bUsedWithNiagaraMeshParticles=true, pixelDepthOffsetMode=PDOM_Legacy, usedShadingModels=Default Lit}
 nodes [
  n1 = ComponentMask(None=ObjectPositionWS()) {b=true}
  n2 = DynamicParameter() {paramNames ["VibrateSpeed(0-00)", "VibrateStrength(0-00)", Param3, Param4]}
  n3 = Add(A=Multiply(A=OneMinus(None=@n1)) {constB=0.0013}, B=Multiply(A=Time(), B=@n2."VibrateSpeed(0-00)"))
  n4 = Add(A=@n3) {constB=0.25}
  n5 = Cosine(None=@n4)
  Add(B=@n5)
  n6 = Sine(None=@n4)
  n7 = AppendVector(A=AppendVector(A=@n6, B=@n5), B=@n6)
  n8 = TextureCoordinate() {coordinateIndex=1}
  n9 = ComponentMask(None=@n8) {r=true}
  n10 = Constant()
  n11 = If(A=ComponentMask(None=@n8) {g=true}, B=Constant() {r=0.375}, "A > B"=@n10, "A == B"=@n10, "A < B"=Constant() {r=1})
  n12 = Multiply(A=Multiply(A=Multiply(A=@n7, B=@n9), B=@n11)) {constB=1.5}
  n13 = Sine(None=@n3)
  n14 = AppendVector(A=AppendVector(A=@n13, B=Cosine(None=@n3)), B=@n13)
  n15 = Multiply(A=Multiply(A=OneMinus(None=@n11), B=Multiply(A=@n9, B=@n14))) {constB=2}
  Reroute(None=Multiply(A=Add(A=@n12, B=@n15), B=Multiply(A=@n2."VibrateStrength(0-00)") {constB=0.5}) {constB=0.5})
  n16 = TextureCoordinate()
  TextureSample(UVs=@n16) {texture=/Game/Athena/Items/QuestInteractables/Tropical_Fruit/Textures/T_TropicalFruit_D}
  TextureSample(UVs=@n16) {samplerType=SAMPLERTYPE_Normal, texture=/Game/Athena/Items/QuestInteractables/Tropical_Fruit/Textures/T_TropicalFruit_N}
 ]
}
